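# Physics asset for the Pilots Backpack cosmetic accessory, giving its tube bones physical bodies and joints against the pack.
# Two capsules shape the pack body, and each of the four dyn_tube bones has one small capsule.
# The tubes form two two-segment chains from the pack (01 to 02, 03 to 04) with wide swing and narrower twist limits.
# A tighter constraint ties the two chain ends together, closing a loop.
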PhysicsAsset /BRCosmetics/Accessories/FORT_Backpacks/M_MED_Pilots_Backpack/Meshes/M_MED_Pilots_Backpack_Physics
bodies {
 dyn_tube_01 [{shapeType=Capsule, length=4, radius=1, center=<3.5 0 0>, rotation=<rot 81.7394 -179.9999 -179.9999>}]
 dyn_tube_02 [{shapeType=Capsule, length=2.5665, radius=1.6212, center=<3 -0.0001 0.4283>, rotation=<rot -85.2952 0.0004 -0.0003>}]
 dyn_tube_03 [{shapeType=Capsule, length=3.3165, radius=1.1212, center=<3.4091 0 0.1752>, rotation=<rot -86.7494 0.0003 -0.0003>}]
 dyn_tube_04 [{shapeType=Capsule, length=3.0665, radius=1.6212, center=<3.1527 0 0.3374>, rotation=<rot -65.6498 -179.9998 179.9999>}]
 pack [
  {shapeType=Capsule, length=16.7255, radius=7.7719, center=<-0.3707 -5.7901 -0.0062>, rotation=<rot 0 90 90>}
  {shapeType=Capsule, length=9.9755, radius=8.7719, center=<-0.3705 10.2531 -10.9816>, rotation=<rot 0 90 90>}
 ]
}
constraints [
 {bone1Name=dyn_tube_02, bone2Name=dyn_tube_01, swing1LimitDegrees=75, swing2LimitDegrees=75, twistLimitDegrees=15}
 {bone1Name=dyn_tube_01, bone2Name=pack, swing1LimitDegrees=75, swing2LimitDegrees=75, twistLimitDegrees=15}
 {bone1Name=dyn_tube_03, bone2Name=pack, swing1LimitDegrees=75, swing2LimitDegrees=75, twistLimitDegrees=15}
 {bone1Name=dyn_tube_04, bone2Name=dyn_tube_03, swing1LimitDegrees=75, swing2LimitDegrees=75, twistLimitDegrees=15}
 {bone1Name=dyn_tube_04, bone2Name=dyn_tube_02, swing1LimitDegrees=10, swing2LimitDegrees=10, twistLimitDegrees=10}
]
